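; Gameplay ability that runs Cube Surfing, a weapon-driven movement mode fuelled by ammo.
; Activation plays a montage whose notifies apply the effect early, then start ammo drain; the local client also starts the movement mode extension and camera modifier.
; Clients play the weapon intro unreplicated to hide lag.
; Ending tears down surfing, commits a duration-dependent cooldown on authority, refills ammo under no-cooldown rules, and launches the player upward unless cancelled.
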(event Ability|EventActivateAbility
  ;[0.01376 0.2962 1] Activate Ability - plays intro anim, then starts movement
  (if (GameplayTags|HasMatchingGameplayTag (Ability|GetActivatingPawn) "(TagName=\"Granted.CubeSurfing.Active\")")
    (Ability|EndAbility)
    (else
      (bind (_asynctaskproxy _notifytag) (Ability|Tasks|PlayMontageAndWaitNotify "None" "/CubeSurfing/Assets/Anim/CubeSurfing_Activate_M.CubeSurfing_Activate_M")
        (:then
          (bind _asb_cube_surfing_weapon (Utilities|Casting|CastToB_CubeSurfing_Weapon (Ability|GetCurrentSourceWeapon)))
          (Ability|HasAuthority
            (:True
              (Variables|/CubeSurfing/Gameplay/Actors/B_CubeSurfing_Weapon.B_CubeSurfing_Weapon_C|SetIsDeploying? true _asb_cube_surfing_weapon))
            (:False
              (Class|BCubeSurfingWeapon|PlayWeaponIntro _asb_cube_surfing_weapon) ; Call non-replicated version on auto-proxy to reduce lag
            ))
          (Ability|RemoveGameplayEffectFromOwnerWithGrantedTags "(GameplayTags=((TagName=\"Gameplay.Status.ImmuneFallingDamage\")))")
          (Ability|RemoveGameplayEffectFromOwnerWithGrantedTags "(GameplayTags=((TagName=\"Gameplay.Status.ImmuneFallingDamage.TillLand\")))")
          (Pawn|Character|ResetFallingHeight (CallFunction|GetAvatarasFortPlayerPawn)))
        (:OnInterrupted
          _)
        (:OnCancelled
          _)
        (:OnNotifyBegin
          (switch Utilities|FlowControl|Switch|SwitchonGameplayTag _notifytag
            (:Animation.Notify.CubeSurf.AddCue
              (Ability|ApplyGameplayEffectToOwner (Variables|Default|GetGE_CubeSurfing)) ; This allows us to add the GE early for cosmetics
            )
            (:Animation.Notify.CubeSurf.Activate
              (CallFunction|EnterMainMovementState)))))
      (Ability|EndAbility))))

(event Custom|Enter%20Main%20Movement%20State
  ;[0 0.5548 1] Set up primary movement state.  To be run after intro animation is ready.
  (Variables|Default|SetAbilityStartTime (Utilities|Time|GetGameTimeinSeconds))
  (bind _output_get (Variables|Default|SetPlayerPawn (CallFunction|GetAvatarasFortPlayerPawn)))
  (bind _asb_cube_surfing_weapon (Utilities|Casting|CastToB_CubeSurfing_Weapon (Weapon|GetCurrentWeapon _output_get)))
  (Variables|Default|SetSurfWeapon _asb_cube_surfing_weapon)
  (CallFunction|InitCubeSurf)
  (Ability|IsLocallyControlled
    (:True
      (Ability|Tasks|WaitGameplayEvent "(TagName=\"Event.CubeSurfing.Stop\")" 0 true
        (:EventReceived
          (Ability|EndAbility))))) ; This event is only received on client anyways
)

(event Custom|PlayCameraModifier
  ;[0.000476 0.8714 1] Play Camera Modifier on surf start
  (bind _returnvalue (Fortnite|Camera|AddCameraModifierforOwningPlayerController "/CubeSurfing/Gameplay/Camera/CameraModifier_CubeSurfing_SurfingStart.CameraModifier_CubeSurfing_SurfingStart_C" (Variables|Default|GetPlayerPawn)))
  (Variables|Default|SetCameraModifierHandle _returnvalue)
  (Utilities|Time|SetTimerbyEvent (delegate Custom|DisableModifier) (Variables|Default|GetCameraModifierDuration)))

(event Custom|StartSurfing
  ;[0.01376 0.2962 1] Start Surfing
  (bind _returnvalue (Class|FortMovementModeExtBlueprintLibrary|ActivateRegisteredMovementModeExtension (Variables|Default|GetMMEClass) (Variables|Default|GetPlayerPawn)))
  (Variables|Default|SetMMERuntimeData _returnvalue))

(event Custom|InitCubeSurf
  ;[0.01376 0.2962 1] Begin Cube Surfing
  (Class|BCubeSurfingWeapon|StartSpendingAmmo (Variables|Default|GetSurfWeapon) (ScalableFloat|GetValueatLevel (Variables|Hotfix|GetHF_AmmoRateSurfing)) true)
  (Ability|IsLocallyControlled
    (:True
      (CallFunction|StartSurfing)
      (CallFunction|PlayCameraModifier))))

(event Custom|DisableModifier
  ;[0.000476 0.8714 1] Play Camera Modifier on surf start
  (Ability|IsLocallyControlled
    (:True
      (CameraModifier|DisableModifier (Variables|Default|GetCameraModifierHandle) true))))

(event Custom|LaunchPlayer
  ;[0.01376 0.2962 1] Launch after leaving
  (Pawn|Character|LaunchCharacterJump (Variables|Default|GetPlayerPawn) "0.000000,0.000000,1100.000000" false true true true))

(event Custom|StopSurfing
  ;[0 0.5548 1] Stop Cube Surf
  (Ability|RemoveGameplayEffectFromOwnerWithAssetTags "(GameplayTags=((TagName=\"GameplayEffect.CherrySmoke.CubeSurf\")))")
  (CallFunction|EndCurrentMME true))

(event Ability|EventOnEndAbility
  (params (bWasCancelled :type Boolean))
  ;[0.01376 0.2962 1] EndAbility - If the ability was cancelled, we know this ability ends from a non-Cube Surfing source. If it was NOT cancelled, it was manually ended by the player.
  ;;[0.5312 0.4731 0.1743 0.5] DeActivate Montage
  (Animation|Montage|PlayMontage (Variables|/Script/Engine.Character|GetMesh (Variables|Default|GetPlayerPawn)) "/CubeSurfing/Assets/Anim/CubeSurfing_Deactivate_M.CubeSurfing_Deactivate_M")
  (CallFunction|StopSurfing)
  (CallFunction|ResetPlayerPitch)
  (CallFunction|CooldownandReload)
  (Utilities|Time|ClearandInvalidateTimerbyHandle (Variables|/CubeSurfing/Gameplay/Actors/B_CubeSurfing_Weapon.B_CubeSurfing_Weapon_C|GetAbilityCostTimer (Variables|Default|GetSurfWeapon)))
  (CallFunction|DisableModifier)
  (if (not bWasCancelled)
    (CallFunction|LaunchPlayer)))

(fn CooldownAndReload
  (Ability|HasAuthority
    (:True
      (Ability|CommitAbilityCooldownWithDuration (Math|Float|SelectFloat (ScalableFloat|GetValueatLevel (LocalVariables|GetDefaultCooldown)) (ScalableFloat|GetValueatLevel (LocalVariables|GetShortCooldown)) (> (- (Utilities|Time|GetGameTimeinSeconds) (Variables|Default|GetAbilityStartTime)) (ScalableFloat|GetValueatLevel (LocalVariables|GetDurationForDefault)))))
      ; If the player has No Cooldowns After Use enabled, the ammo won't be refilled by the cooldown GA so we have to do it now.
      (bind _playerpawn (Variables|Default|GetPlayerPawn)
        (:then
          (if (Inventory|GetHasNoCooldownsfromContext _playerpawn)
            (bind _surfweapon (Variables|Default|GetSurfWeapon)
              (:then
                (if (FortWeapon|Ammo|HasInfiniteReserveEnergy _surfweapon)
                  (FortWeapon|ModAmmo _surfweapon (- (FortWeapon|GetBulletsPerClip _surfweapon) (FortWeapon|GetLocalMagazineAmmoCount _surfweapon))) ; Refill loaded ammo to full without using reserves
                  (else
                    (FortWeapon|TryReloadAmmofromAbility _surfweapon))))
              (:else
                _))
            (return)
            (else
              _)))
        (:else
          _)))
    (:False
      _)))

(fn EndCurrentMME
  (params (EndAll? :type Boolean))
  (bind _playerpawn (Variables|Default|GetPlayerPawn))
  (if (not EndAll?)
    (switch Utilities|FlowControl|Switch|SwitchonECubeSurfingState (Variables|Default|GetCubeState)
      (:NewEnumerator2
        (Utilities|FlowControl|Sequence
          (:then_0
            _)
          (:then_1
            (Class|FortMovementModeExtBlueprintLibrary|EndMovementModeExtensiononActor 0 _playerpawn))))
      (:NewEnumerator0
        _)
      (:NewEnumerator1
        _))
    (Class|FortMovementModeExtBlueprintLibrary|EndMovementModeExtensiononActor (Variables|Default|GetMMERuntimeData) _playerpawn)))

(fn ResetPlayerPitch
  (bind (_relativerotation_roll _relativerotation_pitch _relativerotation_yaw) (Math|Rotator|BreakRotator (Variables|/Script/Engine.SceneComponent|GetRelativeRotation_Roll (Variables|/Script/Engine.Character|GetMesh (Variables|Default|GetPlayerPawn)))))
  (Transformation|SetRelativeRotation (Variables|/Script/Engine.Character|GetMesh (Variables|Default|GetPlayerPawn)) (Math|Rotator|MakeRotator 0.0 _relativerotation_pitch _relativerotation_yaw) false false))

(fn SpendDurability
  (bind _playerpawn (Variables|Default|GetPlayerPawn))
  (bind _returnvalue (Weapon|GetCurrentWeapon _playerpawn))
  ;[0.5289 0.07129 0.8156 0.5] Ensure we still have the Cube Surf weapon equipped before consuming durability
  (if (GameplayTags|HasTag (Tags|GetWeaponDataGameplayTags _returnvalue) "(TagName=\"Item.Weapon.Ranged.CubeSurf\")" true)
    (FortWeapon|UseWeaponDurability(Explicit) _returnvalue (ScalableFloat|GetValueatLevel (Variables|Hotfix|GetHF_DurabilityCost)))))
